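; Quest gameplay ability that tracks horizontal distance traveled by the owning player in a modded vehicle and reports it as quest progress.
; On activation it checks the vehicle for the modded tag, with special-case Taxi and Semi off-road tire tags, then starts a randomized-interval looping timer.
; Each tick accumulates whole meters and sends a stat event past a minimum distance.
; Ending clears the timer and flushes sufficient leftover meters.
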(event Ability|EventActivateAbilityFromEvent
  (params (EventData :type Gameplay%20Event%20Data%20Structure%20%28by%20ref%29))
  ;[1 1 1] Activate ability
  (|CallParentFunction EventData)
  ;;[0 0 0] Taxis and Mudflaps do not appear to apply the Vehicle.Mod tag for tires. Adding a special case
  (if (or (GameplayTags|HasMatchingGameplayTag (Vehicle|GetVehicleAbilitySystemComponent (Vehicle|GetVehicleInterface (Variables|GAVars|Data|GetOwningPlayerPawn))) (Variables|Default|GetModdedVehicleTag)) (or (GameplayTags|HasMatchingGameplayTag (Vehicle|GetVehicleAbilitySystemComponent (Vehicle|GetVehicleInterface (Variables|GAVars|Data|GetOwningPlayerPawn))) "(TagName=\"ValetMods.Tires.Taxi.OffRoad\")") (GameplayTags|HasMatchingGameplayTag (Vehicle|GetVehicleAbilitySystemComponent (Vehicle|GetVehicleInterface (Variables|GAVars|Data|GetOwningPlayerPawn))) "(TagName=\"ValetMods.Tires.Semi.OffRoad\")")))
    (bind _output_get (Variables|Default|SetPreviousLocation (Transformation|GetActorLocation (Variables|GAVars|Data|GetOwningPlayerPawn))))
    (Variables|Default|SetCurrentLocation _output_get)
    (bind _output_get_2 (Variables|Default|SetInterval (Math|Random|RandomFloatinRange 1.7 2.5)))
    (bind _returnvalue (Utilities|Time|SetTimerbyEvent (delegate Custom|SendDistanceUpdate) _output_get_2 true))
    (Variables|Default|SetTImerHandle _returnvalue)))

(event Custom|SendDistanceUpdate
  (bind (_previouslocation_x _previouslocation_y) (Math|Vector|BreakVector (Variables|GetPreviousLocation_X)))
  (bind (_currentlocation_x _currentlocation_y) (Math|Vector|BreakVector (Variables|GetCurrentLocation_X)))
  ;[1 1 1] Send Distance Update
  (Variables|Default|SetCurrentLocation (Transformation|GetActorLocation (Variables|GAVars|Data|GetOwningPlayerPawn)))
  ;;[1 1 1] Cache distance traveled (horizontal only)
  (Variables|Default|SetMetersTraveled (+ (Math|Float|Floor (/ (Math|Vector|VectorLength (Math|Vector|vector-vector (Math|Vector|MakeVector _currentlocation_x _currentlocation_y) (Math|Vector|MakeVector _previouslocation_x _previouslocation_y))) 1e+02)) (Variables|Default|GetMetersTraveled)))
  (Variables|Default|SetPreviousLocation (Variables|Default|GetCurrentLocation)) ; Reset location
  ;;[1 1 1] Send Quest Progress
  (if (> (Variables|Default|GetMetersTraveled) (Variables|Default|GetMinDistance))
    (Quest|SendComplexCustomStatEvent (Variables|GAVars|Data|GetOwningPlayerQuestManager) 0 (Variables|Default|GetT_Quest) "(GameplayTags=)" (Variables|Default|GetMetersTraveled))
    (Variables|Default|SetMetersTraveled 0)))

(event Ability|EventOnEndAbility
  (params (bWasCancelled :type Boolean))
  (bind _timerhandle (Variables|Default|GetTImerHandle))
  ;[1 1 1] On End Ability
  (|CallParentFunction bWasCancelled)
  (if (Utilities|Time|IsValidTimerHandle _timerhandle)
    (Utilities|Time|ClearandInvalidateTimerbyHandle _timerhandle)
    (else
      _))
  (CallFunction|FInalDistanceUpdate))

(event Custom|FInal%20Distance%20Update
  ;[1 1 1] Check if player moved, if so send last update
  (if (>= (Variables|Default|GetMetersTraveled) 10)
    ;[1 1 1] Send Distance Update
    ;;[1 1 1] Send Quest Progress
    (Quest|SendComplexCustomStatEvent (Variables|GAVars|Data|GetOwningPlayerQuestManager) 0 (Variables|Default|GetT_Quest) "(GameplayTags=)" (Variables|Default|GetMetersTraveled))
    (Variables|Default|SetMetersTraveled 0)))
